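# M_Basic_Particle_Flare is an unlit, additive particle-sprite material for glow flares.
# A Mask texture parameter, defaulting to a six-point glow texture, is multiplied by particle color RGB to tint the flare.
# Particle alpha feeds a DepthFade with a parameterized distance to soften intersections with scene geometry.
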Material /Figment_Martin/Effects/Fort_Effects/Materials/Flares/M_Basic_Particle_Flare
graph {
 properties {bUsedWithParticleSprites=true, blendMode=BLEND_Additive, pixelDepthOffsetMode=PDOM_Legacy, shadingModel=MSM_Unlit, usedShadingModels=Unlit}
 nodes [
  n1 = ParticleColor()
  DepthFade(Opacity=@n1.A, FadeDistance=ScalarParameter() {defaultValue=50, parameterName=DepthFade})
  Multiply(A=TextureSampleParameter2D() {parameterName=Mask, texture=/Game/Effects/Fort_Effects/Textures/Flares/T_GlowSar_6pt}.RGB, B=@n1.RGB)
 ]
}
